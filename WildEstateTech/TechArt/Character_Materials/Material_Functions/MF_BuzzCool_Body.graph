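# Material function that optionally replaces incoming emissive with the animated BuzzCool galaxy look for character bodies, via the Buzzcool_Effect static switch.
# Scrolling world-aligned noise and scanline textures build wave and color-mix layers, projected using pre-skinned normals.
# Pre-skinned height gradients, box masks and Fresnel shape purple tinting over desaturated base color, with mask-texture shirt and beard regions.
# Static switches select head handling and view-dependent glowing eyes; time of day scales brightness.
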MaterialFunction /WildEstateTech/TechArt/Character_Materials/Material_Functions/MF_BuzzCool_Body
graph {nodes [
  n1 = FunctionInput() {inputType=FunctionInput_MaterialAttributes}
  n2 = GetMaterialAttributes(None=@n1) {attributeGetTypes [#1, #2]}
  n3 = NamedRerouteUsage()
  FunctionOutput(None=StaticSwitchParameter(
    True=SetMaterialAttributes!(MaterialAttributes=@n2.MaterialAttributes, EmissiveColor=@n3) {attributeSetTypes [#2]}
    False=@n1
   ) {group=GalaxyV2, parameterName=Buzzcool_Effect, sortPriority=50})
  NamedRerouteDeclaration(None=FunctionInput() {inputName=DFL_Mask}) {name=DFL_Mask, nodeColor=<rgba 1 0 0.7969 1>}
  NamedRerouteDeclaration(None=FunctionInput() {inputName=Shader_Gradient}) {name=Shader_Gradient, nodeColor=<rgba 0 1 0.8828 1>}
  n4 = Add(A=NamedRerouteUsage(), B=Multiply(A=Time(), B=ScalarParameter() {defaultValue=1, group=GalaxyV2, parameterName=WavesNoise_Speed_Global}))
  n5 = Multiply(A=ScalarParameter() {defaultValue=0.0725, group=GalaxyV2, parameterName=WavesNoise_RotationSpeed_Global}, B=Time())
  NamedRerouteDeclaration(None=MaterialFunctionCall(
    TextureObject=TextureObject() {samplerType=SAMPLERTYPE_LinearGrayscale, texture=/Game/Athena/SafeZone/PulseLancerAssets/T_Perlin_HQ16_SafeZone}
    TextureSize=ScalarParameter() {defaultValue=1.5, group=GalaxyV2, parameterName=WavesNoise_Tiling_Global}
    WorldPosition=Rotator(Coordinate=@n4, Time=@n5)
    World Space Normal=NamedRerouteUsage()
    ProjectionTransitionContrast=Constant()
   ) {materialFunction=&WorldAlignedTexture}.XYZ Texture) {name=color noise_world, nodeColor=<rgba 0 1 0.2031 1>}
  n6 = LinearInterpolate(Alpha=NamedRerouteUsage()) {constA=1, constB=0}
  n7 = LocalPosition() {localOrigin=InstancePreSkinning}
  n8 = VertexInterpolator(VS=Divide(A=@n7.XYZ, B=PreSkinnedLocalBounds().Half Extents))
  n9 = Multiply(
   A=Subtract(A=@n8.PS, B=ScalarParameter() {group=GalaxyV2, parameterName=Z_Height, sortPriority=50})
   B=ScalarParameter() {defaultValue=1, group=GalaxyV2, parameterName=Z_Str, sortPriority=50}
  )
  n10 = VertexInterpolator(VS=@n7.XYZ)
  n11 = Constant2Vector()
  n12 = AppendVector(
   A=ScalarParameter() {group=GalaxyV2, parameterName=X_Mask, sortPriority=50}
   B=ScalarParameter() {group=GalaxyV2, parameterName=Y_Mask, sortPriority=50}
  )
  n13 = ScalarParameter() {defaultValue=1, group=GalaxyV2, parameterName=Mask_Falloff, sortPriority=50}
  n14 = Saturate(None=LinearInterpolate(
    B=ComponentMask(None=@n9) {b=true}
    Alpha=MaterialFunctionCall!(A=@n10.PS, B=@n11, Bounds=@n12, Edge Falloff=@n13) {materialFunction=/Engine/Functions/Engine_MaterialFunctions01/Gradient/BoxMask-2D}.Result
   ) {constA=1, constB=0})
  NamedRerouteDeclaration(None=Multiply(A=Power(Base=@n6) {constExponent=8}, B=NamedRerouteDeclaration(None=@n14) {name=BGradient, nodeColor=<rgba 0.9453 1 0 1>})) {name=CharGradient, nodeColor=<rgba 1 0.0469 0 1>}
  NamedRerouteDeclaration(None=VertexInterpolator(VS=PreSkinnedNormal()).PS) {name=skeletal normal, nodeColor=<rgba 1 0.0234 0 1>}
  n15 = Multiply(A=ScalarParameter() {defaultValue=1, group=GalaxyV2, parameterName=Galaxy_GlobalBrightness, sliderMax=1}, B=NamedRerouteUsage())
  n16 = Multiply(
   A=Multiply(A=NamedRerouteUsage()) {constB=10}
   B=FunctionInput(Preview=Constant()) {bUsePreviewValueAsDefault=true, inputName=eye Mask, inputType=FunctionInput_Scalar, sortPriority=5}
  )
  n17 = StaticSwitchParameter(True=LinearInterpolate(A=@n15, B=NamedRerouteUsage(), Alpha=Saturate(None=@n16)), False=@n15) {group=BuzzCool_Eyes, parameterName=Eyes}
  n18 = Constant() {r=0.8}
  n19 = MaterialFunctionCall(Value_Morning=@n18, Value_Day=Constant() {r=1}, Value_Evening=@n18, Value_Night=Constant() {r=0.7}) {materialFunction=&Scalar_Time_of_Day}
  NamedRerouteDeclaration(None=LinearInterpolate(A=Multiply(A=@n17, B=@n19.Result), B=NamedRerouteUsage(), Alpha=NamedRerouteUsage())) {name=emissive output, nodeColor=<rgba 1 0.6797 0 1>}
  n20 = TextureSampleParameter2D() {group=BuzzCool_Eyes, parameterName=BuzzCool_Mask}
  NamedRerouteDeclaration(None=@n20.R) {name=shirt mask, nodeColor=<rgba 1 0.0703 0 1>}
  NamedRerouteDeclaration(None=@n8.PS) {name=skeletal gradient, nodeColor=<rgba 0.5234 0 1 1>}
  n21 = VectorParameter() {group=GalaxyV2, parameterName=Base Purple, defaultValue=<rgba 0.3741 0.2313 0.9271 1>}
  n22 = Multiply(A=@n21.RGB, B=ScalarParameter() {defaultValue=0.4, group=GalaxyV2, parameterName=Purple_Dark}) {constB=0.4}
  NamedRerouteDeclaration(None=@n22) {name=dark_purple, nodeColor=<rgba 1 0 0.0937 1>}
  n23 = VectorParameter() {group=BuzzCool_Eyes, parameterName=BuzzCool_Eyes_BaseEmissive, defaultValue=<rgba 0.0262 0.9823 0 1>}
  n24 = VectorParameter() {group=BuzzCool_Eyes, parameterName=BuzzCool_Eyes_CoreEmissive, defaultValue=<rgba 0.9939 1 0.1359 1>}
  n25 = DotProduct(A=CameraVectorWS(), B=PixelNormalWS())
  n26 = ScalarParameter() {defaultValue=3, group=BuzzCool_Eyes, parameterName=eye Pow}
  n27 = Saturate(None=Multiply(A=Power(Base=@n25, Exp=@n26), B=ScalarParameter() {defaultValue=1, group=BuzzCool_Eyes, parameterName=eye Multi}))
  NamedRerouteDeclaration(None=Multiply(
    A=LinearInterpolate(A=@n23.RGB, B=@n24.RGB, Alpha=@n27)
    B=ScalarParameter() {defaultValue=1, group=BuzzCool_Eyes, parameterName=Eye_EmissiveIntensity}
   )) {name=Eyes, nodeColor=<rgba 0 1 0.1852 1>}
  NamedRerouteDeclaration(None=@n20.B) {name=Beard_Mask, nodeColor=<rgba 1 0 0.1406 1>}
  NamedRerouteDeclaration(None=TextureObject() {texture=/Game/Global/Textures/Default/Masks_and_Gradients/Scanlines_Offset}) {name=NoiseTex, nodeColor=<rgba 0.4766 0 1 1>}
  n28 = NamedRerouteUsage()
  n29 = Multiply(A=NamedRerouteUsage(), B=ScalarParameter() {defaultValue=0.5, group=GalaxyV2, parameterName=Noise_World_Multiply}) {constB=0.5}
  n30 = Time()
  n31 = ScalarParameter() {group=secondaryNoise, parameterName=Secondary_noise_Speed}
  n32 = MaterialFunctionCall(
   TextureObject=TextureObject() {samplerType=SAMPLERTYPE_Masks, texture=/Game/Global/Textures/Default/Noise/WindNoise_Mask_1}
   TextureSize=ScalarParameter() {defaultValue=1, group=secondaryNoise, parameterName=secondary_noise_Scale}
   WorldPosition=Add(A=NamedRerouteUsage(), B=Multiply(A=@n30, B=@n31))
   World Space Normal=NamedRerouteUsage()
   ProjectionTransitionContrast=Constant()
  ) {materialFunction=&WorldAlignedTexture}
  n33 = ScalarParameter() {group=secondaryNoise, parameterName=Secondary_noise_Str}
  n34 = MaterialFunctionCall(Color=Add(A=Add(A=@n28, B=@n29), B=Multiply(A=@n32.XYZ Texture, B=@n33))) {materialFunction=&SplitComponents}
  n35 = Time()
  n36 = ScalarParameter() {defaultValue=0.005, group=GalaxyV2, parameterName=BottomFresnel_World_GradientSpeed_X}
  n37 = MaterialFunctionCall(
   TextureObject=NamedRerouteUsage()
   TextureSize=ScalarParameter() {defaultValue=0.5, group=GalaxyV2, parameterName=BottomFresnel_World_Tiling}
   WorldPosition=MaterialFunctionCall(
    R=Panner(Coordinate=@n34.R, Time=@n35, Speed=@n36)
    G=Panner(Coordinate=@n34.G, Time=@n35, Speed=@n36)
    B=Panner(Coordinate=@n34.B, Time=@n35, Speed=@n36)
   ) {materialFunction=&AppendMany}.RGB
   World Space Normal=NamedRerouteUsage()
   ProjectionTransitionContrast=Constant()
  ) {materialFunction=&WorldAlignedTexture}
  n38 = ComponentMask(None=Divide(A=NamedRerouteUsage(), B=ScalarParameter() {defaultValue=35, group=BuzzCool_Waves, parameterName=erase_Mask})) {b=true}
  n39 = LinearInterpolate(
   A=ScalarParameter() {group=GalaxyV2, parameterName=Color_Mix_Str}
   Alpha=TextureSample(UVs=Panner(Coordinate=@n38) {speedY=0.1}) {samplerType=SAMPLERTYPE_LinearColor, texture=/Game/Athena/SafeZone/Lines}.RGB
  ) {constB=0}
  NamedRerouteDeclaration(None=Power(Base=Multiply(A=@n37.XY Texture, B=@n39), Exp=ScalarParameter() {defaultValue=2, group=GalaxyV2, parameterName=Color_Mix_Pow})) {name=ColorMix, nodeColor=<rgba 0.125 0 1 1>}
  n40 = NamedRerouteUsage()
  n41 = ScalarParameter() {defaultValue=0.9, group=GalaxyV2, parameterName=BuzzCool_BaseColor_Intensity, sliderMax=1}
  n42 = Power(Base=OneMinus(None=NamedRerouteUsage()))
  n43 = LinearInterpolate(
   B=Desaturation(None=Multiply(A=@n40, B=@n41), Fraction=ScalarParameter() {group=GalaxyV2, parameterName=BuzzCool_BaseColor_Desaturate, sliderMax=1})
   Alpha=Saturate(None=LinearInterpolate(Alpha=@n42) {constA=1, constB=0.6})
  )
  n44 = NamedRerouteUsage()
  n45 = ScalarParameter() {defaultValue=1, group=GalaxyV2, parameterName=Beard_Color_Str}
  n46 = NamedRerouteUsage()
  n47 = LinearInterpolate(
   A=ScalarParameter() {defaultValue=1, group=GalaxyV2, parameterName=Shirt_Purple_Str_Dark}
   B=ScalarParameter() {defaultValue=1, group=GalaxyV2, parameterName=Shirt_Purple_Str}
   Alpha=NamedRerouteUsage()
  )
  n48 = StaticSwitchParameter(
   True=LinearInterpolate(A=@n43, B=NamedRerouteUsage(), Alpha=Multiply(A=@n44, B=@n45))
   False=LinearInterpolate(A=@n43, B=NamedRerouteUsage(), Alpha=Multiply(A=@n46, B=@n47))
  ) {group=GalaxyV2, parameterName=Head?}
  n49 = Multiply(
   A=Subtract(A=NamedRerouteUsage(), B=ScalarParameter() {group=GalaxyV2, parameterName=Purple_Color_Height, sortPriority=50})
   B=ScalarParameter() {defaultValue=1, group=GalaxyV2, parameterName=Purple_Gradient_Str, sortPriority=50}
  )
  n50 = LinearInterpolate(A=@n22, B=@n21.RGB, Alpha=Saturate(None=ComponentMask(None=@n49) {b=true}))
  n51 = LinearInterpolate(
   A=NamedRerouteDeclaration(None=@n50) {name=Purple, nodeColor=<rgba 1 0.1406 0 1>}
   B=Multiply(A=@n50, B=ScalarParameter() {defaultValue=1.5, group=GalaxyV2, parameterName=Purple_Highlight}) {constB=1.5}
   Alpha=Saturate(None=NamedRerouteUsage())
  )
  n52 = LinearInterpolate(A=Multiply(A=ScalarParameter() {defaultValue=0.6, parameterName=Pants_ShadowStr}, B=@n51), B=@n51, Alpha=NamedRerouteUsage())
  n53 = OneMinus(None=NamedRerouteUsage())
  n54 = OneMinus(None=DotProduct(A=CameraVectorWS(), B=PixelNormalWS()))
  n55 = ScalarParameter() {group=GalaxyV2, parameterName=purple_layer1_Str}
  n56 = LinearInterpolate(
   A=LinearInterpolate(A=LinearInterpolate(A=@n48, B=@n52, Alpha=@n53), B=@n52, Alpha=@n53)
   B=NamedRerouteUsage()
   Alpha=Add(A=Multiply(A=@n53, B=@n54), B=Multiply(A=@n53, B=@n55))
  )
  n57 = NamedRerouteUsage()
  n58 = NamedRerouteUsage()
  n59 = Multiply(A=@n58, B=ScalarParameter() {defaultValue=0.2, group=BuzzCool_Waves, parameterName=Shirt_Waves_Str})
  n60 = NamedRerouteUsage()
  n61 = LinearInterpolate(
   A=@n56
   B=MaterialFunctionCall(Base=@n56, Blend=Desaturation(None=@n57)) {materialFunction=&Blend_Overlay}.Result
   Alpha=StaticSwitchParameter(True=@n58, False=LinearInterpolate(A=@n58, B=@n59, Alpha=@n60) {constB=0.2}) {group=GalaxyV2, parameterName=Head?}
  )
  n62 = NamedRerouteUsage()
  n63 = Saturate(None=LinearInterpolate(Alpha=Multiply(A=@n62) {constB=8}) {constA=0.2})
  NamedRerouteDeclaration(None=LinearInterpolate(A=LinearInterpolate(A=@n56, B=@n61, Alpha=@n63) {constAlpha=1}, B=@n56, Alpha=NamedRerouteUsage())) {name=GalaxyEmissive, nodeColor=<rgba 1 0.4219 0 1>}
  n64 = ScalarParameter() {defaultValue=3, group=GalaxyV2, parameterName=Galaxy_FresEX, sliderMax=8}
  NamedRerouteDeclaration(None=Saturate(None=Fresnel(ExponentIn=@n64) {baseReflectFraction=0})) {name=Master Fresnel, nodeColor=<rgba 0 1 0.3906 1>}
  NamedRerouteDeclaration(None=@n2.EmissiveColor) {name=emissive in, nodeColor=<rgba 0 1 1 1>}
  n65 = NamedRerouteUsage()
  NamedRerouteDeclaration(None=Saturate(None=Desaturation(None=@n65))) {name=WavesColor, nodeColor=<rgba 1 0.4688 0 1>}
  n66 = NamedRerouteUsage()
  n67 = Multiply(A=@n65, B=Saturate(None=Add(A=@n66) {constB=0.55}))
  n68 = Fresnel() {baseReflectFraction=0, exponent=2}
  NamedRerouteDeclaration(None=Multiply(A=LinearInterpolate(A=@n67, Alpha=@n68) {constB=0}, B=ScalarParameter() {defaultValue=0.3, group=BuzzCool_Waves, parameterName=Waves_Str})) {name=WavesA, nodeColor=<rgba 0.1484 0 1 1>}
  NamedRerouteDeclaration(None=@n10.PS) {name=skeletal position, nodeColor=<rgba 1 0 0.9844 1>}
  NamedRerouteUsage()
  NamedRerouteUsage()
  TextureCoordinate() {uTiling=8, vTiling=8}
  TextureObject() {texture=/Game/Global/Textures/Default/Caustics/T_Turbulence_Seamless}
  TextureObject() {samplerType=SAMPLERTYPE_LinearGrayscale, texture=/Game/Effects/Fort_Effects/Textures/Tile/T_Caustic_Wind_Diagonal}
  TextureObject() {samplerType=SAMPLERTYPE_LinearGrayscale, texture=&T_Perlin_HQ16_SafeZone}
  TextureObject() {samplerType=SAMPLERTYPE_LinearGrayscale, texture=/Game/Athena/SafeZone/PulseLancerAssets/T_GradientLine_HQ16_SafeZone}
  TextureObject() {samplerType=SAMPLERTYPE_LinearColor, texture=&Lines}
 ]}
