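# Material function that masks a character's DF lines around the eyelids as they close, behind a static switch.
# Blink, lid-down and squint pose parameters drive sphere mask radii for upper and lower lids, evaluated on pre-skinned local position.
# Left-side centres mirror the right's in X; an X-sign step picks the side, and an eye-area mask limits the effect.
# A Debug output exposes the masking.
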MaterialFunction /WildEstateTech/TechArt/Character_Materials/Material_Functions/MF_WildEstate_EyeballsDFLineMasking
graph {nodes [
  n1 = FunctionInput() {inputName=DFLines, inputType=FunctionInput_Scalar}
  n2 = NamedRerouteUsage()
  FunctionOutput(None=StaticSwitchParameter(True=LinearInterpolate(A=@n1, Alpha=@n2) {constAlpha=0, constB=0}, False=@n1) {group=Eyelid DF Lines Masking, parameterName=UseEyelidDFLinesMasking, sortPriority=-1})
  n3 = StaticSwitchParameter(True=NamedRerouteUsage(), False=NamedRerouteUsage()) {group=Eyelid DF Lines Masking, parameterName=DebugRightlidSphere, sortPriority=-1}
  FunctionOutput(None=@n3) {outputName=Debug}
  n4 = NamedRerouteUsage()
  n5 = VectorParameter() {group=Eyelid DF Lines Masking, parameterName=R_Upperlid_SphereCenter, defaultValue=<rgba -4.8 8.5 167.8 1>}
  n6 = ScalarParameter() {defaultValue=2, group=Eyelid DF Lines Masking, parameterName=SphereMin}
  n7 = ScalarParameter() {defaultValue=7, group=Eyelid DF Lines Masking, parameterName=SphereMax}
  n8 = LinearInterpolate(A=@n6, B=@n7, Alpha=NamedRerouteUsage())
  n9 = ScalarParameter() {defaultValue=100, group=Eyelid DF Lines Masking, parameterName=Hardness}
  n10 = NamedRerouteUsage()
  n11 = VectorParameter() {group=Eyelid DF Lines Masking, parameterName=R_Lowerlid_SphereCenter, defaultValue=<rgba -5.5 7 158 1>}
  n12 = ScalarParameter() {defaultValue=2, group=Eyelid DF Lines Masking, parameterName=SphereMin}
  n13 = ScalarParameter() {defaultValue=7, group=Eyelid DF Lines Masking, parameterName=SphereMax}
  n14 = LinearInterpolate(A=@n12, B=@n13, Alpha=NamedRerouteUsage())
  n15 = NamedRerouteDeclaration(None=Max(
    A=SphereMask(A=@n4, B=@n5.RGB, Radius=@n8, Hardness=@n9) {attenuationRadius=7, hardnessPercent=75}
    B=SphereMask(A=@n10, B=@n11.RGB, Radius=@n14, Hardness=@n9) {attenuationRadius=7, hardnessPercent=75}
   )) {name=RightlidsMasking, nodeColor=<rgba 0 1 0.2266 1>}
  n16 = Multiply(A=@n5.RGB, B=Constant3Vector() {constant=<rgba -1 1 1 0>})
  n17 = LinearInterpolate(A=@n6, B=@n7, Alpha=NamedRerouteUsage())
  n18 = Multiply(A=@n11.RGB, B=Constant3Vector() {constant=<rgba -1 1 1 0>})
  n19 = LinearInterpolate(A=@n12, B=@n13, Alpha=NamedRerouteUsage())
  n20 = NamedRerouteDeclaration(None=Max(
    A=SphereMask(A=@n4, B=@n16, Radius=@n17, Hardness=@n9) {attenuationRadius=7, hardnessPercent=75}
    B=SphereMask(A=@n10, B=@n18, Radius=@n19, Hardness=@n9) {attenuationRadius=7, hardnessPercent=75}
   )) {name=LeftLidMasking, nodeColor=<rgba 1 0 0.0937 1>}
  n21 = NamedRerouteUsage()
  NamedRerouteDeclaration(None=Multiply(A=NamedRerouteUsage(), B=LinearInterpolate(A=@n15, B=@n20, Alpha=@n21))) {name=DFLinesMask_Final, nodeColor=<rgba 0 1 0.6953 1>}
  n22 = ScalarParameter() {group=01_EyeAnimation_Poses, parameterName=R_blink_pose, sliderMax=1, sortPriority=2}
  n23 = ScalarParameter() {group=01_EyeAnimation_Poses, parameterName=R_squint_inner_pose, sliderMax=1, sortPriority=6}
  n24 = ScalarParameter() {defaultValue=0.35, group=01_EyeAnimation_Parameters, parameterName=Squint_Eyelid_Influence, sliderMax=1, sortPriority=31}
  n25 = Max(
   A=ScalarParameter() {group=01_EyeAnimation_Poses, parameterName=R_upperLidDown_pose, sliderMax=1, sortPriority=4}
   B=Max(A=@n22, B=Multiply(A=@n23, B=@n24) {constB=0.5})
  )
  n26 = ScalarParameter() {defaultValue=1, group=01_EyeAnimation_Parameters, parameterName=AnimationThreshold_Eyes, sliderMax=8, sortPriority=30}
  NamedRerouteDeclaration(None=Saturate(None=Power(Base=@n25, Exp=@n26))) {name=R_upperLid_close_pose, nodeColor=<rgba 0.125 1 0 1>}
  n27 = ScalarParameter() {group=01_EyeAnimation_Poses, parameterName=L_blink_pose, sliderMax=1, sortPriority=1}
  n28 = ScalarParameter() {group=01_EyeAnimation_Poses, parameterName=L_squint_inner_pose, sliderMax=1, sortPriority=5}
  n29 = Max(A=@n27, B=@n28)
  NamedRerouteDeclaration(None=Saturate(None=Power(Base=@n29, Exp=@n26))) {name=L_lowerLid_close_pose, nodeColor=<rgba 0.125 1 0 1>}
  n30 = Max(
   A=ScalarParameter() {group=01_EyeAnimation_Poses, parameterName=L_upperLidDown_pose, sliderMax=1, sortPriority=3}
   B=Max(A=Multiply(A=@n28, B=@n24) {constB=0.5}, B=@n27)
  )
  NamedRerouteDeclaration(None=Saturate(None=Power(Base=@n30, Exp=@n26))) {name=L_upperLid_close_pose, nodeColor=<rgba 0.125 1 0 1>}
  n31 = NamedRerouteUsage()
  n32 = VectorParameter() {group=Eyelid DF Lines Masking, parameterName=R_EyeCenter, defaultValue=<rgba -6 10.481 164 1>}
  n33 = ScalarParameter() {defaultValue=6, group=Eyelid DF Lines Masking, parameterName=EyeMasks_Radius}
  n34 = Multiply(A=@n32.RGB, B=Constant3Vector() {constant=<rgba -1 1 1 0>})
  NamedRerouteDeclaration(None=Max(A=SphereMask(A=@n31, B=@n32.RGB, Radius=@n33) {attenuationRadius=6}, B=SphereMask(A=@n31, B=@n34, Radius=@n33) {attenuationRadius=6})) {name=EyeAreaMask, nodeColor=<rgba 1 0.375 0 1>}
  n35 = Max(A=@n22, B=@n23)
  NamedRerouteDeclaration(None=Saturate(None=Power(Base=@n35, Exp=@n26))) {name=R_lowerLid_close_pose, nodeColor=<rgba 0.125 1 0 1>}
  n36 = LocalPosition() {localOrigin=InstancePreSkinning}
  MaterialExpressionNamedRerouteDeclaration_9 = NamedRerouteDeclaration(None=ShaderStageSwitch(PixelShader=VertexInterpolator(VS=@n36.XYZ).PS, VertexShader=@n36.XYZ)) {name=PreskinnedLocPos, nodeColor=<rgba 0.3125 1 0 1>}
  n37 = Convert(None=@MaterialExpressionNamedRerouteDeclaration_9) {
   nodeName=Break Float3
   convertInputs [{
     type=Vector3
     expressionInput {
      expression=&self:MaterialExpressionNamedRerouteDeclaration_9
      inputName=None
      mask=0
      maskA=0
      maskB=0
      maskG=0
      maskR=0
      outputIndex=0
     }
    }]
   convertOutputs [{type=Scalar}, {type=Scalar}, {type=Scalar}]
  }
  NamedRerouteDeclaration(None=Step(X=@n37) {constX=0}) {name=LeftEyeMask, nodeColor=<rgba 0.5703 1 0 1>}
  SetMaterialAttributes!(MaterialAttributes, EmissiveColor=@n3) {attributeSetTypes [#1]}
 ]}
